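; Server-side disguise helper that builds a unique list of every outfit worn in the match, NPCs included.
; On BeginPlay it listens to the game state's stateful phase-updated event, and stops on EndPlay.
; At the SafeZones phase it converts each player pawn's character cosmetic to swap data in GeneratedOutfitData and prints the list.
; In editor or PIE, a duplicate DebugInPIE path repeats this population on a delayed loop for testing.
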(event EventEndPlay
  (params (EndPlayReason :type EEndPlayReason%20Enum :default "Destroyed"))
  (bind _gamephaseupdatedasynctask (Variables|Default|GetGamePhaseUpdatedAsyncTask)
    (:then
      (GameplayEventRouter|StopListeningtoStatefulEvent _gamephaseupdatedasynctask))
    (:else)))

(event EventBeginPlay
  ;[1 1 1] Begin Play - Bind to game phase changed
  (Networking|SwitchHasAuthority
    (:Authority
      (bind _returnvalue (GameplayEventRouter|GetEventRouter "/Script/FortniteGame.FortGameState" self))
      (bind (_asynctaskproxy _payload) (GameplayEventRouter|StartListeningToStatefulEvent _returnvalue "/Script/FortniteGame.FortGamePhaseUpdatedEvent")
        (:then
          (Variables|Default|SetGamePhaseUpdatedAsyncTask _asynctaskproxy)
          ;[1 1 1] Helps with PIE Testing
          (if (Editor|InEditororPIE)
            (Development|PrintString "Disguise (PIE Active)")
            (Utilities|FlowControl|Delay 5.0)
            (CallFunction|DebuginPIE)
            (Utilities|ForLoopwithDelay 0 6 15.0
              (:LoopBody
                (CallFunction|DebuginPIE)))))
        (:OnEventReceived
          _)
        (:OnSavedEventStateExists
          _))
      (CallFunction|GamePhaseChanged (Utilities|Struct|BreakFortGamePhaseUpdatedEvent _payload)))))

(event Custom|GamePhaseChanged
  (params (GamePhase :type EAthenaGamePhase%20Enum))
  (Development|PrintString (Utilities|String|Append "Game Phase Changed: " (Utilities|String|EnumtoString GamePhase)) false)
  ;[1 1 1] Wait until after the bus phase to populate the list
  (if (Utilities|Enum|Equal(Enum) GamePhase "SafeZones")
    ;[1 1 1] Populate list of all outfits being used in this match (includes NPCs)
    (bind _outfortplayerpawns (Fortnite|GetFortPlayerPawns))
    (for _array_element _outfortplayerpawns
      (bind (_cosmeticloadout_bannericonid _cosmeticloadout_bannercolorid _cosmeticloadout_skydivecontrail _cosmeticloadout_glider _cosmeticloadout_pickaxe _cosmeticloadout_character) (Utilities|Struct|BreakFortAthenaLoadout (Variables|/Script/FortniteGame.FortPlayerPawn|GetCosmeticLoadout_BannerIconId _array_element)))
      (Utilities|Array|AddUnique (Variables|Default|GetGeneratedOutfitData) (Class|FortCosmeticFunctionLibrary|MakeSwapDataforCosmetic _cosmeticloadout_character)))
    (for _array_element_2 (Variables|Default|GetGeneratedOutfitData)
      (bind _array_element_item (Utilities|Struct|BreakFortSwapItemAndVariantData _array_element_2))
      (Development|PrintString (Utilities|GetDisplayName _array_element_item)))))

(event Custom|DebugInPIE
  ;[1 1 1] Populate list of all outfits being used in this match (includes NPCs)
  (bind _outfortplayerpawns (Fortnite|GetFortPlayerPawns))
  (for _array_element _outfortplayerpawns
    (bind (_cosmeticloadout_bannericonid _cosmeticloadout_bannercolorid _cosmeticloadout_skydivecontrail _cosmeticloadout_glider _cosmeticloadout_pickaxe _cosmeticloadout_character) (Utilities|Struct|BreakFortAthenaLoadout (Variables|/Script/FortniteGame.FortPlayerPawn|GetCosmeticLoadout_BannerIconId _array_element)))
    (Utilities|Array|AddUnique (Variables|Default|GetGeneratedOutfitData) (Class|FortCosmeticFunctionLibrary|MakeSwapDataforCosmetic _cosmeticloadout_character)))
  (for _array_element_2 (Variables|Default|GetGeneratedOutfitData)
    (bind _array_element_item (Utilities|Struct|BreakFortSwapItemAndVariantData _array_element_2))
    (Development|PrintString (Utilities|GetDisplayName _array_element_item))))

(fn ConstructionScript)
